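; Time-of-day Blueprint for the Asteria HarvestHomes Standing Lamp B environment prop.
; GetTimeOfDayBlueprintDefaultVariables returns a fixed struct of default settings, presumably read by shared time-of-day logic outside this file.
; The construction script and an extra placeholder function, NewFunction, are both empty, so the asset holds no other logic of its own.
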(fn GetTimeOfDayBlueprintDefaultVariables
  (return (Utilities|Struct|MakeTimeOfDayBlueprintDefaultVariables 0 2.0 false true false true)))

(fn NewFunction)

(fn ConstructionScript)
